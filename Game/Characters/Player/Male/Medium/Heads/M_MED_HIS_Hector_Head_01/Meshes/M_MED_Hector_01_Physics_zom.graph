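# Physics asset for the zombie variant of the Hector 01 male medium player head mesh, judging by the "zom" suffix.
# It defines two bodies, C_jaw and faceAttach, each a zero-length capsule that acts as a sphere, with the larger one offset on the jaw.
# A single constraint joins the jaw to faceAttach with equal angle values for both swing axes and twist.
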PhysicsAsset /Game/Characters/Player/Male/Medium/Heads/M_MED_HIS_Hector_Head_01/Meshes/M_MED_Hector_01_Physics_zom
bodies {C_jaw [{shapeType=Capsule, length=0, radius=5, center=<0.1883 5 -4.4809>}], faceAttach [{shapeType=Capsule, length=0, radius=1}]}
constraints [{bone1Name=C_jaw, bone2Name=faceAttach, swing1LimitDegrees=45, swing2LimitDegrees=45, twistLimitDegrees=45}]
